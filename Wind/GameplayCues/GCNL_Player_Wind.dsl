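; Looping gameplay cue for a player in wind that drives a Niagara effect and layers movement-specific wind cues on the pawn.
; On application it binds to skydiving begin and end events, swapping between local skydiving and ground-movement cues guarded by applied flags.
; Removal unbinds and removes the active cue.
; On loop start, the source wind spline goes to Niagara, and a 0.1-second timer feeds in wind direction and attenuation at the player's location.
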(event EventOnApplication(EitherVFX-preferthisone)
  (params (MyTarget :type Actor%20Object%20Reference) (Parameters :type Gameplay%20Cue%20Parameters%20Structure%20%28by%20ref%29) (ParticleComponents :type Array%20of%20FXSystem%20Component%20Object%20References) (AudioComponents :type Array%20of%20Audio%20Component%20Object%20References) (BurstCameraShakeInstance :type Camera%20Shake%20Base%20Object%20Reference) (BurstDecalInstance :type Decal%20Actor%20Object%20Reference))
  (Variables|Default|SetGameplayCueParams Parameters)
  (bind _asfort_player_pawn (Utilities|Casting|CastToFortPlayerPawn MyTarget))
  (bind _output_get (Variables|Default|SetFort%20Player%20Pawn _asfort_player_pawn))
  (EventDispatchers|AssignOnBeginSkydivingEvent _output_get (EventDispatchers|CreateEvent :!event Custom|SkydivingStarted))
  (EventDispatchers|AssignOnEndSkydivingEvent (Variables|Default|GetFort%20Player%20Pawn) (EventDispatchers|CreateEvent :!event Custom|SkyDivingEnded))
  (if (Parachute|IsSkydiving (Variables|Default|GetFort%20Player%20Pawn))
    (CallFunction|OnBeginSkydiving)
    (else
      (CallFunction|OnBeginGroundMovement))))

(event EventOnRemoval(EitherVFX-preferthisversion)
  (params (MyTarget :type Actor%20Object%20Reference) (Parameters :type Gameplay%20Cue%20Parameters%20Structure%20%28by%20ref%29) (ParticleComponents :type Array%20of%20FXSystem%20Component%20Object%20References) (AudioComponents :type Array%20of%20Audio%20Component%20Object%20References) (BurstCameraShakeInstance :type Camera%20Shake%20Base%20Object%20Reference) (BurstDecalInstance :type Decal%20Actor%20Object%20Reference))
  (bind _fort_player_pawn (Variables|Default|GetFort%20Player%20Pawn))
  (Utilities|IsValid _fort_player_pawn
    (:Is%20Valid
      (EventDispatchers|UnbindEventfromOnBeginSkydivingEvent _fort_player_pawn (EventDispatchers|CreateEvent :!event Custom|SkydivingStarted))
      (EventDispatchers|UnbindEventfromOnEndSkydivingEvent (Variables|Default|GetFort%20Player%20Pawn) (EventDispatchers|CreateEvent :!event Custom|SkyDivingEnded))
      (if (Parachute|IsSkydiving (Variables|Default|GetFort%20Player%20Pawn))
        (CallFunction|OnEndSkydiving)
        (else
          (CallFunction|OnEndGroundMovement))))))

(event Custom|OnEndGroundMovement
  (if (Variables|Default|GetIsGroundMovementCueApplied)
    (Projectile|RemoveGameplayCue%28Looping%29Local (Variables|Default|GetFort%20Player%20Pawn) (Variables|Default|GetGroundMovementGameplayCueTag) (Variables|Default|GetGameplayCueParams))
    (Variables|Default|SetIsGroundMovementCueApplied false)))

(event Custom|OnBeginSkydiving
  (if (not (Variables|Default|GetIsSkydivingCueApplied))
    (Projectile|AddGameplayCue%28Looping%29Local (Variables|Default|GetFort%20Player%20Pawn) (Variables|Default|GetSkydivingGameplayCueTag) (Variables|Default|GetGameplayCueParams))
    (Variables|Default|SetIsSkydivingCueApplied true)))

(event Custom|SkydivingStarted
  (CallFunction|OnBeginSkydiving)
  (CallFunction|OnEndGroundMovement))

(event Custom|SkyDivingEnded
  (CallFunction|OnEndSkydiving)
  (CallFunction|OnBeginGroundMovement))

(event Custom|OnBeginGroundMovement
  (if (not (Variables|Default|GetIsGroundMovementCueApplied))
    (Projectile|AddGameplayCue%28Looping%29Local (Variables|Default|GetFort%20Player%20Pawn) (Variables|Default|GetGroundMovementGameplayCueTag) (Variables|Default|GetGameplayCueParams))
    (Variables|Default|SetIsGroundMovementCueApplied true)))

(event Custom|OnEndSkydiving
  (if (Variables|Default|GetIsSkydivingCueApplied)
    (Projectile|RemoveGameplayCue%28Looping%29Local (Variables|Default|GetFort%20Player%20Pawn) (Variables|Default|GetSkydivingGameplayCueTag) (Variables|Default|GetGameplayCueParams))
    (Variables|Default|SetIsSkydivingCueApplied false)))

(event EventOnLoopingStart(EitherVFX-preferthisversion)
  (params (MyTarget :type Actor%20Object%20Reference) (Parameters :type Gameplay%20Cue%20Parameters%20Structure%20%28by%20ref%29) (ParticleComponents :type Array%20of%20FXSystem%20Component%20Object%20References) (AudioComponents :type Array%20of%20Audio%20Component%20Object%20References))
  (bind _gameplaycueparams (Variables|Default|GetGameplayCueParams))
  (bind (_normalizedmagnitude _rawmagnitude _effectcontext _matchedtagname _originaltag _aggregatedsourcetags _aggregatedtargettags _location _normal _instigator _effectcauser _sourceobject) (Ability|GameplayCue|BreakGameplayCueParameters _gameplaycueparams))
  (Variables|Default|SetPlayerRef MyTarget)
  (bind _result (ArrayMacros|SafeArrayGet ParticleComponents)
    (:%20%20
      (bind _asniagara_particle_system_component (Utilities|Casting|CastToNiagaraComponent _result))
      (bind _output_get (Variables|Default|SetNiagaraSys _asniagara_particle_system_component))
      (Niagara|SetNiagaraVariable(Object) _output_get "SplineObject" _sourceobject)
      (Utilities|Time|SetTimerbyEvent (EventDispatchers|CreateEvent :!event Custom|UpdateVFX) 0.1 true)
      (bind _aswind_spline_component (Utilities|Casting|CastToWindSplineComponent _sourceobject))
      (Variables|Default|SetWindSplineComponent _aswind_spline_component)
      (Game|BindEventtoOnEndPlay (delegate Custom|OnEndPlay_Event)))))

(event Custom|OnEndPlay_Event
  (params (Actor :type Actor%20Object%20Reference) (EndPlayReason :type EEndPlayReason%20Enum :default "Destroyed"))
  (Utilities|Time|ClearandInvalidateTimerbyHandle 0 (Variables|Default|GetFxUpdateTimer)))

(event Custom|UpdateVFX
  (bind _niagarasys (Variables|Default|GetNiagaraSys))
  (bind (_outdirection _outattenuation) (Class|WindSplineComponent|GetWindDataatLocation (Variables|Default|GetWindSplineComponent) (Transformation|GetActorLocation (Variables|Default|GetPlayerRef))))
  (Niagara|SetNiagaraVariable(Vector3) _niagarasys "Direction" _outdirection)
  (Niagara|SetNiagaraVariable(Float) _niagarasys "Attenuation" _outattenuation))

(fn UpdateFX
  (params (NewParam :type Wind%20Spline%20Standard%20Child%20Object%20Reference)))

(fn ConstructionScript)
