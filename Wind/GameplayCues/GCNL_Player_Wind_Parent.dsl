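; Parent looping gameplay cue for a player in wind, feeding live wind data to its looping audio.
; On loop start, for a locally viewed pawn, it caches that pawn, the wind spline component from the cue's source object, and an audio component.
; Each tick it samples the spline at the pawn's location and sets audio parameters for wind angle, player facing angle and attenuation.
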(event EventTick
  (params (DeltaSeconds :type Float%20%28single-precision%29))
  (bind _targetpawn (Variables|Default|GetTargetPawn)
    (:then
      (bind _windsplinecomponent (Variables|Default|GetWindSplineComponent)
        (:then
          (bind (_outdirection _outattenuation) (Class|WindSplineComponent|GetWindDataatLocation _windsplinecomponent (Transformation|GetActorLocation _targetpawn)))
          (bind _audiocomponent (Variables|Default|GetAudioComponent)
            (:then
              (Audio|Parameter|SetFloatParameter _audiocomponent "WindAngle" (Fortnite|VectortoNormalizedAngleinDegrees _outdirection) :!class /Script/AudioExtensions.AudioParameterControllerInterface)
              (Audio|Parameter|SetFloatParameter _audiocomponent "PlayerAngle" (Fortnite|VectortoNormalizedAngleinDegrees (Transformation|GetActorForwardVector _targetpawn)) :!class /Script/AudioExtensions.AudioParameterControllerInterface)
              (Audio|Parameter|SetFloatParameter _audiocomponent "Attenuation" _outattenuation :!class /Script/AudioExtensions.AudioParameterControllerInterface))
            (:else)))
        (:else)))
    (:else)))

(event EventOnLoopingStart(EitherVFX-preferthisversion)
  (params (MyTarget :type Actor%20Object%20Reference) (Parameters :type Gameplay%20Cue%20Parameters%20Structure%20%28by%20ref%29) (ParticleComponents :type Array%20of%20FXSystem%20Component%20Object%20References) (AudioComponents :type Array%20of%20Audio%20Component%20Object%20References))
  (bind (_normalizedmagnitude _rawmagnitude _effectcontext _matchedtagname _originaltag _aggregatedsourcetags _aggregatedtargettags _location _normal _instigator _effectcauser _sourceobject) (Ability|GameplayCue|BreakGameplayCueParameters Parameters))
  (bind _asfort_pawn (Utilities|Casting|CastToFortPawn MyTarget))
  (if (Pawn|IsLocallyViewed _asfort_pawn)
    (Variables|Default|SetTargetPawn _asfort_pawn)
    (bind _aswind_spline_component (Utilities|Casting|CastToWindSplineComponent _sourceobject))
    (Variables|Default|SetWindSplineComponent _aswind_spline_component)
    (bind _result (ArrayMacros|SafeArrayGet AudioComponents)
      (:%20%20
        (Variables|Default|SetAudioComponent _result)))))

(event EventOnRemoval(EitherVFX-preferthisversion)
  (params (MyTarget :type Actor%20Object%20Reference) (Parameters :type Gameplay%20Cue%20Parameters%20Structure%20%28by%20ref%29) (ParticleComponents :type Array%20of%20FXSystem%20Component%20Object%20References) (AudioComponents :type Array%20of%20Audio%20Component%20Object%20References) (BurstCameraShakeInstance :type Camera%20Shake%20Base%20Object%20Reference) (BurstDecalInstance :type Decal%20Actor%20Object%20Reference))
  (if (== MyTarget (Variables|Default|GetTargetPawn))
    _))

(fn ConstructionScript)
